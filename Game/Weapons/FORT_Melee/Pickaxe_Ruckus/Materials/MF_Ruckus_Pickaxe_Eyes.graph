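# Material function that draws animated eyes into the emissive channel of the Ruckus pickaxe material.
# Cropped UVs place both eyes, one mirrored, then shift by sine-driven glances, a camera-based reflection vector and hash jitter, with overridable clamp limits.
# The pupil is a texture mask or a sphere mask sized by camera distance.
# It multiplies the saturated sum of incoming emissive and two circular sclera masks.
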MaterialFunction /Game/Weapons/FORT_Melee/Pickaxe_Ruckus/Materials/MF_Ruckus_Pickaxe_Eyes
graph {nodes [
  n1 = VectorParameter() {group=RuckusEyes, parameterName=AnatomicRightIrisTextureCrop, defaultValue=<rgba 0.045 0.195 0.16 0.305>}
  AppendVector(A=@n1.RGB, B=@n1.RGB)
  n2 = VectorParameter() {group=RuckusEyes, parameterName=AnatomicLeftIrisTextureCrop, defaultValue=<rgba 0.045 0.025 0.16 0.155>}
  AppendVector(A=@n2.RGB, B=@n2.RGB)
  Constant() {r=0.05}
  Constant() {r=2}
  n3 = GetMaterialAttributes(None=FunctionInput() {inputType=FunctionInput_MaterialAttributes}) {attributeGetTypes [#1]}
  n4 = TextureCoordinate()
  n5 = VectorParameter() {
   group=RuckusEyes
   parameterName=AnatomicLeftEyeValues
   channelNames {a=Disabled, b=Scale, g=ShiftUpDown, r=ShiftLeftRight}
   defaultValue=<rgba 0.115 0.11 0.07 0>
  }
  n6 = MaterialFunctionCall(
   X=Subtract(A=@n5.ShiftLeftRight, B=@n5.ShiftLeftRight)
   Y=Subtract(A=@n5.ShiftUpDown, B=@n5.ShiftUpDown)
   Z=Add(A=@n5.ShiftLeftRight, B=@n5.ShiftLeftRight)
   A=Add(A=@n5.ShiftUpDown, B=@n5.ShiftUpDown)
  ) {materialFunction=&MakeFloat4}
  n7 = MaterialFunctionCall!(
   UVs=@n4
   TextureIn
   UpperLeftCorner=ComponentMask(None=@n6.Result) {g=true, r=true}
   LowerRightCorner=ComponentMask(None=@n6.Result) {a=true, b=true}
   ExportFloat4
  ) {materialFunction=/Engine/Functions/Engine_MaterialFunctions01/Texturing/TextureCropping}
  n8 = VectorParameter() {
   group=RuckusEyes
   parameterName=AnatomicRightEyeValues
   channelNames {a=ScaleLeftRight, b=ScaleUpDown, g=ShiftUpDown, r=ShiftLeftRight}
   defaultValue=<rgba 0.115 0.13 0.08 0>
  }
  n9 = MaterialFunctionCall(
   X=Subtract(A=@n8.ShiftLeftRight, B=@n8.ShiftLeftRight)
   Y=Subtract(A=@n8.ShiftUpDown, B=@n8.ShiftUpDown)
   Z=Add(A=@n8.ShiftLeftRight, B=@n8.ShiftLeftRight)
   A=Add(A=@n8.ShiftUpDown, B=@n8.ShiftUpDown)
  ) {materialFunction=&MakeFloat4}
  n10 = MaterialFunctionCall(UVs=@n4, UpperLeftCorner=ComponentMask(None=@n9.Result) {g=true, r=true}, LowerRightCorner=ComponentMask(None=@n9.Result) {a=true, b=true}) {materialFunction=&TextureCropping}
  n11 = MaterialFunctionCall(Float2=@n10.Crop UVs) {materialFunction=&BreakOutFloat2Components}
  n12 = AppendVector(A=OneMinus(None=@n11.R), B=@n11.G)
  n13 = MaterialFunctionCall(Float2=Add(A=Multiply(A=@n7.Crop UVs, B=@n7.Crop UVs), B=Multiply(A=@n12, B=@n10.Crop Mask))) {materialFunction=&BreakOutFloat2Components}
  n14 = MaterialFunctionCall!(
   Normal=Constant3Vector() {constant=<rgba 0 0 150 0>}
   CameraVector=Transform(None=CameraVectorWS()) {transformSourceType=TRANSFORMSOURCE_World, transformType=TRANSFORM_Local}
  ) {materialFunction=/Engine/Functions/Engine_MaterialFunctions02/Utility/CustomReflectionVector}
  n15 = ComponentMask(None=@n14.Reflection Vector) {r=true}
  n16 = Time()
  n17 = ScalarParameter() {defaultValue=0.2, group=RuckusEyes, parameterName=LookLeftRight_Rate}
  n18 = ConstantBiasScale(None=Sine(None=Multiply(A=@n16, B=@n17) {constA=0.7})) {bias=0}
  n19 = Constant() {r=10}
  n20 = ScalarParameter() {defaultValue=-0.15, group=RuckusEyes, parameterName=LookDistanceMin}
  n21 = ScalarParameter() {defaultValue=0.15, group=RuckusEyes, parameterName=LookDistanceMax}
  n22 = FunctionInput() {bUsePreviewValueAsDefault=true, description="R = Min, G = Max, B = Lerp (1 = use this input)", inputName=LookDistanceOverride}
  n23 = MaterialFunctionCall(Float2=LinearInterpolate(
    A=MaterialFunctionCall(X=@n20, Y=@n21) {materialFunction=&MakeFloat2}.Result
    B=ComponentMask(None=@n22) {g=true, r=true}
    Alpha=ComponentMask(None=@n22) {b=true}
   )) {materialFunction=&BreakOutFloat2Components}
  n24 = Clamp(None=Add(A=Multiply(A=@n15), B=Multiply(A=@n18, B=@n19) {constA=0.7}), Min=@n23.R, Max=@n23.R) {maxDefault=0.2, minDefault=-0.2}
  n25 = ScalarParameter() {defaultValue=7, group=RuckusEyes, parameterName=JitterUpDown}
  n26 = Multiply(A=MaterialFunctionCall(Float=Multiply(A=@n16, B=@n25) {constA=0.7}) {materialFunction=&MF_Hash}.Float) {constB=0.02}
  n27 = Time()
  n28 = ScalarParameter() {defaultValue=0.4, group=RuckusEyes, parameterName=LookUpDown_Rate}
  n29 = ConstantBiasScale(None=Sine(None=Multiply(A=@n27, B=@n28) {constA=0.7})) {bias=0}
  n30 = Add(A=ComponentMask(None=@n14.Reflection Vector) {b=true}) {constB=0.5}
  n31 = LinearInterpolate(
   A=Clamp(None=Multiply(A=@n29, B=@n19) {constA=0.7}, Min=@n23.R, Max=@n23.R) {maxDefault=0.2, minDefault=-0.2}
   Alpha=Saturate(None=Multiply(A=@n30) {constB=0.5})
  ) {constB=0}
  n32 = AppendVector(A=Add(A=@n13.R, B=Add(A=@n24, B=@n26)), B=Add(A=@n13.G, B=Add(A=@n31, B=@n26)))
  n33 = TextureSampleParameter2D(UVs=@n32) {group=RuckusEyes, parameterName=Param, texture=/Game/Characters/Player/Female/Medium/Bodies/F_MED_Genius/Textures/T_F_MED_Genius_Eye_D}
  n34 = Constant2Vector() {g=0.5, r=0.5}
  n35 = ScalarParameter() {defaultValue=100, group=NeonDepth, parameterName=DF_FarDistance}
  n36 = Subtract(A=Distance(A=WorldPosition().XYZ, B=CameraPositionWS()), B=@n35)
  n37 = Subtract(A=ScalarParameter() {defaultValue=500, group=NeonDepth, parameterName=DF_NearDistance}, B=@n35)
  n38 = LinearInterpolate(
   A=ScalarParameter() {defaultValue=0.05, parameterName=DF_PupilSizeNear}
   B=ScalarParameter() {defaultValue=0.12, parameterName=DF_PupilSizeFar}
   Alpha=Saturate(None=Divide(A=@n36, B=@n37))
  )
  n39 = StaticSwitchParameter(True=@n33.R, False=OneMinus(None=SphereMask(A=@n32, B=@n34, Radius=@n38))) {group=RuckusEyes, parameterName=UseTextureMaskForIris}
  n40 = TextureCoordinate()
  n41 = Constant() {r=0.06}
  n42 = Add(
   A=SphereMask(A=@n40, B=Constant2Vector() {g=0.085, r=0.1}, Radius=@n41)
   B=SphereMask(A=@n40, B=Constant2Vector() {g=0.245, r=0.1}, Radius=@n41)
  )
  n43 = ScalarParameter() {defaultValue=1, group=RuckusEyes, parameterName=ScleraOpacityMultiply}
  n44 = Saturate(None=Add(A=@n3.EmissiveColor, B=Multiply(A=@n42, B=@n43)) {constB=0})
  FunctionOutput(None=SetMaterialAttributes!(MaterialAttributes=@n3.MaterialAttributes, EmissiveColor=Multiply(A=@n39, B=@n44)) {attributeSetTypes [#1]})
  n45 = Fmod(A=@n16, B=Constant() {r=5})
  n46 = Constant()
  If(A=@n45, B=Constant() {r=1}, "A > B"=@n45, "A == B"=@n46, "A < B"=@n46)
  n47 = ScalarParameter() {defaultValue=5, group=RuckusEyes, parameterName=JitterLeftRight}
  Multiply(A=MaterialFunctionCall(Float=Multiply(A=@n16, B=@n47) {constA=0.7}) {materialFunction=&MF_Hash}.Float) {constB=0.02}
  Reroute(None=Add(A=@n7.Crop Mask, B=@n10.Crop Mask))
 ]}
